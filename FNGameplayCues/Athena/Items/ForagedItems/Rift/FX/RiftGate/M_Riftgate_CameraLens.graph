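# Translucent, unlit, two-sided material for Niagara mesh particles, named as the Rift Gate camera-lens effect, drawing animated tinted energy noise.
# Panning caustic and plasma textures, their UVs warped by scrolling directional noise, drive three tint colors.
# That color is multiplied by static caustics and particle color, then passed through MF_FX_EmissiveAdjustments.
# Alpha combines plasma noise, particle alpha and a V-gradient mask confining it to a band near high V.
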Material /FNGameplayCues/Athena/Items/ForagedItems/Rift/FX/RiftGate/M_Riftgate_CameraLens
graph {
 properties {
  bUsedWithNiagaraMeshParticles=true
  blendMode=BLEND_Translucent
  pixelDepthOffsetMode=PDOM_Legacy
  shadingModel=MSM_Unlit
  twoSided=true
  usedShadingModels=Unlit
 }
 nodes [
  n1 = TextureCoordinate()
  n2 = TextureCoordinate() {vTiling=2}
  n3 = Multiply(A=TextureSample(UVs=Panner(Coordinate=@n2) {speedY=0.2}) {samplerType=SAMPLERTYPE_LinearGrayscale, texture=/Game/Effects/Fort_Effects/Emotes/UltraLight/Texture/T_DirectionalNoise_Greyscale}.R) {constB=0.05}
  n4 = TextureCoordinate() {uTiling=2, vTiling=2}
  n5 = Constant()
  n6 = ScalarParameter() {defaultValue=-1, parameterName=Noise_Pan_Y}
  n7 = TextureSample(UVs=Panner(Coordinate=Add(A=@n3, B=@n4) {constB=0}, Speed=AppendVector(A=@n5, B=@n6)) {speedY=-1}) {samplerType=SAMPLERTYPE_LinearGrayscale, texture=/Game/Global/Textures/Default/Caustics/T_Nosie_Caustics_Grayscale}
  n8 = TextureCoordinate() {uTiling=3}
  n9 = Constant()
  n10 = ScalarParameter() {defaultValue=-0.3, parameterName=Distortion_Noise_Pan_Y}
  n11 = TextureSampleParameter2D(UVs=Panner(Coordinate=Add(A=@n8, B=@n3) {constB=0}, Speed=AppendVector(A=@n9, B=@n10)) {speedY=-0.3}) {
   parameterName=NoiseTexture
   samplerType=SAMPLERTYPE_LinearGrayscale
   texture=/Game/Effects/Fort_Effects/Emotes/UltraLight/Texture/T_Plasma_C_Greyscale
  }
  n12 = Add(
   A=Multiply(A=VectorParameter() {parameterName=Color_A, defaultValue=<rgba 0 1.9534 2 1>}.RGB, B=Power(Base=@n7.R))
   B=LinearInterpolate(
    A=VectorParameter() {parameterName=Color_B, defaultValue=<rgba 0 0.9313 1 1>}.RGB
    B=VectorParameter() {parameterName=Color_C, defaultValue=<rgba 0 1.5 0.2228 1>}.RGB
    Alpha=Power(Base=@n11.R) {constExponent=1}
   )
  )
  n13 = ParticleColor()
  MaterialFunctionCall(Emissive=Multiply(
    A=TextureSample(UVs=@n1) {samplerType=SAMPLERTYPE_LinearGrayscale, texture=&T_Nosie_Caustics_Grayscale}.R
    B=Multiply(A=@n12, B=@n13.RGB)
   )) {materialFunction=&MF_FX_EmissiveAdjustments}
  n14 = Add(A=Power(Base=@n11.R) {constExponent=1}) {constB=0}
  n15 = TextureCoordinate()
  n16 = Power(Base=OneMinus(None=ComponentMask(None=@n15) {g=true}))
  n17 = ComponentMask(None=TextureCoordinate()) {g=true}
  n18 = Saturate(None=Multiply(A=Power(Base=@n17) {constExponent=25}) {constB=20})
  n19 = Saturate(None=Multiply(A=Multiply(A=@n16, B=@n18)) {constB=20})
  Multiply(A=Multiply(A=@n13.A, B=Multiply(A=@n14, B=@n19)))
 ]
}
