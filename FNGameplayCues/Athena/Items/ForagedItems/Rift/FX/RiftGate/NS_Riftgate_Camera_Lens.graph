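# Niagara camera-lens effect for the Rift gate, showing a brief mesh flash and post-process pulse.
# Two lens emitters burst a mesh that grows, fades and animates dynamic material parameters.
# Two post-process emitters use component renderers to briefly pulse blend weight.
# An Overload float compared to a threshold sets spawn probability to zero or one, selecting the normal or overload pair, the latter with a brighter lens.
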NiagaraSystem /FNGameplayCues/Athena/Items/ForagedItems/Rift/FX/RiftGate/NS_Riftgate_Camera_Lens
dependencies {
 usedDataInterfaces [&NiagaraDataInterfaceCurve, &NiagaraDataInterfaceColorCurve]
 usedDynamicInputs []
 usedModules [
  &SystemState
  &self:SystemSpawnScript.SystemScriptSource.SystemScriptGraph.NiagaraNodeAssignment_0.SetVariables_#1
 ]
}
emitters {
 Camera_Lens {
  emitterUpdateScript [
   &EmitterState(Life Cycle Mode=::NewEnumerator1, Loop Behavior=::NewEnumerator1, Loop Delay=0.05, Loop Duration=1, UseLoopDelay=true)
   &SpawnBurst_Instantaneous(
    Spawn Count=1
    Spawn Probability=&MakeCustomFloatFromBool(Bool=&SetBoolByFloatComparison(A=$System.Overload, B=0.8), True Float=0, False Float=1)
    Use Spawn Probability=true
   )
  ]
  particleSpawnScript [
   &InitializeParticle(
    Color=<rgba 3 3 3 3>
    Lifetime=0.3
    Mesh Scale=&Multiply_VectorByFloat(Vector=<1 1 1>, Float=0.37)
    Mesh Scale Mode=::NewEnumerator4
    UsePositionOffset=false
    Write Material Random=true
   )
   &InitialMeshOrientation(Rotation=<0 0 0.25>, Use Orientation Vector=true, Use Rotation Vector=true)
  ]
  particleUpdateScript [
   &ParticleState()
   &ScaleMeshSize(Scale Factor=&Multiply_VectorByFloat(
     Vector=<1 1 1>
     Float=&FloatFromCurve(FloatCurve=DataInterface() {propertyValues {Curve {keys [{tangentMode=RCTM_User, time=0, value=0.9}, {tangentMode=RCTM_User, time=1, value=1.2}]}}})
    ))
   &DynamicMaterialParameters(
    Index 0 Param 2=&FloatFromCurve(FloatCurve=DataInterface() {propertyValues {Curve {keys [<at 0 1>, {arriveTangent=1, interpMode=RCIM_Cubic, leaveTangent=1, time=0.2566, value=1.6894}, <at 1 2>]}}})
    Index 0 Param 3=1
    Index 0 Param 4=1
    Module.Param0WriteEnabled=true
    Module.Param1WriteEnabled=true
    Module.Param2WriteEnabled=true
    Module.Param3WriteEnabled=true
   )
   &ScaleColor(
    Scale Alpha=&FloatFromCurve(
     FloatCurve=DataInterface() {propertyValues {Curve {keys [<at 0 0 RCIM_Cubic>, {arriveTangent=-1, leaveTangent=-1, time=0.1, value=1}, <at 0.6 1>, <at 1 0>]}}}
     Scale Curve=&FNTimeOfDayScalar(Night Scale=1, Morning Scale=1, Evening Scale=1)
    )
    Scale RGB=&MakeVectorFromLinearColorRGB(Color=&ColorFromCurve(ColorCurve=DataInterface() {propertyValues {
        AlphaCurve {keys []}
        BlueCurve {keys [
          <at 0.0115 0.2615>
          {arriveTangent=1.1173, interpMode=RCIM_Cubic, leaveTangent=1.1173, time=0.3418, value=1}
          {arriveTangent=2.0008, interpMode=RCIM_Cubic, leaveTangent=2.0008, time=0.6725, value=1}
          <at 0.8416 2>
         ]}
        GreenCurve {keys [
          <at 0.0115 1>
          {arriveTangent=-0.707, interpMode=RCIM_Cubic, leaveTangent=-0.707, time=0.3418, value=1}
          {arriveTangent=-0.0094, interpMode=RCIM_Cubic, leaveTangent=-0.0094, time=0.6725, value=0.5327}
          <at 0.8416 0.9953>
         ]}
        RedCurve {keys [
          <at 0.0115 0>
          <at 0.3418 1 RCIM_Cubic>
          {arriveTangent=-2.0008, interpMode=RCIM_Cubic, leaveTangent=-2.0008, time=0.6725, value=0}
          <at 0.8416 0>
         ]}
       }}))
    ScaleA=true
    ScaleRGB=false
   )
  ]
  rendererClasses [&NiagaraMeshRendererProperties]
  renderers [{rendererClass=&NiagaraMeshRendererProperties, rendererIndex=0}]
 }
 Camera_Lens_Overload {
  emitterUpdateScript [
   &EmitterState(Life Cycle Mode=::NewEnumerator1, Loop Behavior=::NewEnumerator1, Loop Delay=0.05, Loop Duration=1, UseLoopDelay=true)
   &SpawnBurst_Instantaneous(
    Spawn Count=1
    Spawn Probability=&MakeCustomFloatFromBool(Bool=&SetBoolByFloatComparison(A=$System.Overload, B=0.8), True Float=1, False Float=0)
    Use Spawn Probability=true
   )
  ]
  particleSpawnScript [
   &InitializeParticle(
    Color=<rgba 6 6 6 3>
    Lifetime=0.3
    Mesh Scale=&Multiply_VectorByFloat(Vector=<1 1 1>, Float=0.37)
    Mesh Scale Mode=::NewEnumerator4
    UsePositionOffset=false
    Write Material Random=true
   )
   &InitialMeshOrientation(Rotation=<0 0 0.25>, Use Orientation Vector=true, Use Rotation Vector=true)
  ]
  particleUpdateScript [
   &ParticleState()
   &ScaleMeshSize(Scale Factor=&Multiply_VectorByFloat(
     Vector=<1 1 1>
     Float=&FloatFromCurve(FloatCurve=DataInterface() {propertyValues {Curve {keys [{tangentMode=RCTM_User, time=0, value=0.9}, {tangentMode=RCTM_User, time=1, value=1.2}]}}})
    ))
   &DynamicMaterialParameters(
    Index 0 Param 2=&FloatFromCurve(FloatCurve=DataInterface() {propertyValues {Curve {keys [<at 0 1>, {arriveTangent=1, interpMode=RCIM_Cubic, leaveTangent=1, time=0.2566, value=1.6894}, <at 1 2>]}}})
    Index 0 Param 3=1
    Index 0 Param 4=1
    Module.Param0WriteEnabled=true
    Module.Param1WriteEnabled=true
    Module.Param2WriteEnabled=true
    Module.Param3WriteEnabled=true
   )
   &ScaleColor(
    Scale Alpha=&FloatFromCurve(
     FloatCurve=DataInterface() {propertyValues {Curve {keys [<at 0 0 RCIM_Cubic>, {arriveTangent=-1, leaveTangent=-1, time=0.1, value=1}, <at 0.6 1>, <at 1 0>]}}}
     Scale Curve=&FNTimeOfDayScalar(Night Scale=1, Morning Scale=1, Evening Scale=1)
    )
    Scale RGB=&MakeVectorFromLinearColorRGB(Color=&ColorFromCurve(ColorCurve=DataInterface() {propertyValues {
        AlphaCurve {keys []}
        BlueCurve {keys [
          <at 0.0115 0.2615>
          {arriveTangent=1.1173, interpMode=RCIM_Cubic, leaveTangent=1.1173, time=0.3418, value=1}
          {arriveTangent=2.0008, interpMode=RCIM_Cubic, leaveTangent=2.0008, time=0.6725, value=1}
          <at 0.8416 2>
         ]}
        GreenCurve {keys [
          <at 0.0115 1>
          {arriveTangent=-0.707, interpMode=RCIM_Cubic, leaveTangent=-0.707, time=0.3418, value=1}
          {arriveTangent=-0.0094, interpMode=RCIM_Cubic, leaveTangent=-0.0094, time=0.6725, value=0.5327}
          <at 0.8416 0.9953>
         ]}
        RedCurve {keys [
          <at 0.0115 0>
          <at 0.3418 1 RCIM_Cubic>
          {arriveTangent=-2.0008, interpMode=RCIM_Cubic, leaveTangent=-2.0008, time=0.6725, value=0}
          <at 0.8416 0>
         ]}
       }}))
    ScaleA=true
    ScaleRGB=false
   )
  ]
  rendererClasses [&NiagaraMeshRendererProperties]
  renderers [{rendererClass=&NiagaraMeshRendererProperties, rendererIndex=0}]
 }
 PostProcess {
  emitterUpdateScript [
   &EmitterState(Life Cycle Mode=::NewEnumerator1, Loop Behavior=::NewEnumerator1, Loop Delay=0.1, Loop Duration=1.2567, UseLoopDelay=true)
   &SpawnBurst_Instantaneous(
    Spawn Probability=&MakeCustomFloatFromBool(Bool=&SetBoolByFloatComparison(A=$System.Overload, B=0.8), True Float=0, False Float=1)
    Use Spawn Probability=true
    UseLoopCountLimit=false
   )
  ]
  particleSpawnScript [/Niagara/Modules/Spawn/Initialization/InitializeParticle(
    Lifetime=0.5
    Mesh Scale=<5 5 5>
    Module.Write Color=true
    Module.Write Lifetime=true
    Module.Write Mass=true
    Module.Write Material Random=true
    Module.Write Position=false
    Module.Write Scale=false
    Module.Write SpriteRotation=false
    Module.Write SpriteSize=false
   )]
  particleUpdateScript [
   &ParticleState()
   /Game/Effects/Niagara/Modules/ComponentRenderer/FN_PostProcess(Blend:NiagaraFloat=&FloatFromCurve(FloatCurve=DataInterface() {propertyValues {Curve {keys [
         {arriveTangent=1.2408, interpMode=RCIM_Cubic, leaveTangent=1.2408, tangentMode=RCTM_User, time=0, value=0}
         {arriveTangent=0.0476, interpMode=RCIM_Cubic, leaveTangent=0.0476, tangentMode=RCTM_User, time=0.1, value=0.1}
         {arriveTangent=-0.0405, interpMode=RCIM_Cubic, leaveTangent=-0.0405, tangentMode=RCTM_User, time=0.5, value=0}
        ]}}}))
  ]
  rendererClasses [&NiagaraComponentRendererProperties]
  renderers [{rendererClass=&NiagaraComponentRendererProperties, rendererIndex=0}]
 }
 PostProcessOverload {
  emitterUpdateScript [
   &EmitterState(Life Cycle Mode=::NewEnumerator1, Loop Behavior=::NewEnumerator1, Loop Delay=0.1, Loop Duration=1.2567, UseLoopDelay=true)
   &SpawnBurst_Instantaneous(
    Spawn Probability=&MakeCustomFloatFromBool(Bool=&SetBoolByFloatComparison(A=$System.Overload, B=0.8), True Float=1, False Float=0)
    Use Spawn Probability=true
    UseLoopCountLimit=false
   )
  ]
  particleSpawnScript [&InitializeParticle(
    Lifetime=0.5
    Mesh Scale=<5 5 5>
    Module.Write Color=true
    Module.Write Lifetime=true
    Module.Write Mass=true
    Module.Write Material Random=true
    Module.Write Position=false
    Module.Write Scale=false
    Module.Write SpriteRotation=false
    Module.Write SpriteSize=false
   )]
  particleUpdateScript [
   &ParticleState()
   &FN_PostProcess(Blend:NiagaraFloat=&FloatFromCurve(FloatCurve=DataInterface() {propertyValues {Curve {keys [
         {arriveTangent=1.2408, interpMode=RCIM_Cubic, leaveTangent=1.2408, tangentMode=RCTM_User, time=0, value=0}
         {arriveTangent=0.0476, interpMode=RCIM_Cubic, leaveTangent=0.0476, tangentMode=RCTM_User, time=0.1, value=0.1}
         {arriveTangent=-0.0405, interpMode=RCIM_Cubic, leaveTangent=-0.0405, tangentMode=RCTM_User, time=0.5, value=0}
        ]}}}))
  ]
  rendererClasses [&NiagaraComponentRendererProperties]
  renderers [{rendererClass=&NiagaraComponentRendererProperties, rendererIndex=0}]
 }
}
summary {
 systemName=NS_Riftgate_Camera_Lens
 emitters [
  {emitterName=Camera_Lens, rendererClasses [&NiagaraMeshRendererProperties]}
  {emitterName=Camera_Lens_Overload, rendererClasses [&NiagaraMeshRendererProperties]}
  {emitterName=PostProcess, rendererClasses [&NiagaraComponentRendererProperties]}
  {emitterName=PostProcessOverload, rendererClasses [&NiagaraComponentRendererProperties]}
 ]
 userVariables [User.Overload {type=NiagaraFloat, defaultValue=0, description=""}]
}
